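; Exterior Secret Facility vault alarm that registers with its vault door and starts or stops on activation events.
; A looping timer flip-flops UpdateAlarmVisuals, which extends the parent to toggle the point light and fade out audio when inactive.
; On world ready, the authority links it to environment below, retrying up to a limit to cover load order in PIE.
; If that environment is destroyed, audio fades and the alarm expires shortly after.
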(event EventBeginPlay
  (CallFunction|RegisterAlarms))

(event Custom|RegisterAlarms
  (Class|BPILargeDoorTrigger|RegisterVaultDoorAlarm(Message) (Variables|Default|GetVaultDoorActor) self))

(event EventActivateVaultAlarms
  (CallFunction|StartAlarm))

(event EventDeactivateVaultAlarms
  (CallFunction|EndAlarm))

(event Custom|StartAlarm
  (bind _returnvalue (Utilities|Time|SetTimerbyEvent (delegate Custom|TurnOnAndOff) (Variables|Default|GetAlarmFrequency) true))
  (Variables|Default|SetAlarmTimerHandle _returnvalue)
  (Development|PrintString "Large Door Vault - Alarm Triggered!" true true "(R=1.000000,G=0.000000,B=0.000000,A=1.000000)" 5.0))

(event Custom|TurnOnAndOff
  (Utilities|FlowControl|FlipFlop
    (:A
      (CallFunction|UpdateAlarmVisuals false true))
    (:B
      (CallFunction|UpdateAlarmVisuals false))))

(event Custom|EndAlarm
  (Utilities|Time|ClearandInvalidateTimerbyHandle (Variables|Default|GetAlarmTimerHandle))
  (CallFunction|UpdateAlarmVisuals true)
  (Development|PrintString "Turn alarm off"))

(event EventOnWorldReady
  ;[0 1 0.135] EventOnWorldReady
  ; Try to link to the environment when the World is Ready.
  (|CallParentFunction)
  (Networking|SwitchHasAuthority
    (:Authority
      (CallFunction|TrytoLinktoEnvironment))))

(event Custom|TryToLinkToEnvironment
  (bind _linkattemptsmax (Variables|LinkToEnvironment|GetLinkAttemptsMax))
  ;[0.3798 0.6987 1] TryToLinkToEnvironment
  ; Attempt to attach the alarm to the environment to prevent it from floating if the env is destroyed
  (bind _returnvalue (LinkToActorComponent|PerformLinkingCheck (Variables|Default|GetFortLinkToActor) "Down" (Variables|LinkToEnvironment|GetEnvironmentLinkedActor)))
  ;;[1 0 0.8947] Fallback handles case where the alarm which is in a content bundle in Asteria is loaded before the environment piece it is trying to attach to (in a level instance) - !Not expected to be an issue outside of PIE!
  (bind _result (Math|Integer|IncrementInt (Variables|LinkToEnvironment|GetLinkAttemptsCount))
    (:%20%20
      (if (not (or _returnvalue (>= (Variables|LinkToEnvironment|GetLinkAttemptsCount) _linkattemptsmax)))
        (Fortnite|FortShippingLog self (Utilities|String|Append "HEIST ALARM: Attempt to Link to Environment Failed on attempt #" (Utilities|String|ToString(Integer) _result) " of " (Utilities|String|ToString(Integer) _linkattemptsmax) ". This is not expected outside of PIE"))
        (bind _returnvalue_2 (Utilities|Time|SetTimerbyEvent (EventDispatchers|CreateEvent :!event Custom|TryToLinkToEnvironment) (Variables|LinkToEnvironment|GetLinkAttemptTiming))) ; try to link again
        (Variables|LinkToEnvironment|SetTimerH_RetryLinkToEnvironment _returnvalue_2)))))

(event OnLinkedActorDestroyed(FortLinkToActor)
  (params (DamageCauser :type Actor%20Object%20Reference))
  ;[0.4149 0.164 0] OnLinkedActorDestroyed
  ; Set the life span to destroy the alarm when the attached envionment actor is destoyed
  (bind _audio (Variables|Default|GetAudio)
    (:then
      (Audio|Components|Audio|FadeOut _audio 0.1 0.0 "Linear" :!class /Script/Engine.AudioComponent)
      (Components|Activation|Deactivate _audio))
    (:else
      _))
  (Actor|SetLifeSpan 0.2))

(event EventEndPlay
  (params (EndPlayReason :type EEndPlayReason%20Enum :default "Destroyed"))
  ;[1 0.3416 0] EndPlay
  ; Cleanup- clear timers, unbind events, etc.
  (Utilities|FlowControl|Sequence
    (:then_0
      (Utilities|Time|ClearandInvalidateTimerbyHandle (Variables|LinkToEnvironment|GetTimerH_RetryLinkToEnvironment))
      (EventProcessorSubsystem|StopListeningToFilteredEvent (Variables|Default|GetStartAlarmVerbHandle)))
    (:then_1
      (|CallParentFunction EndPlayReason) ; call parent EndPlay last in case of \norder of operations dependencies
    )))

(fn GetTimeOfDayBlueprintDefaultVariables
  (return (Utilities|Struct|MakeTimeOfDayBlueprintDefaultVariables 0 0.0 (not (Variables|/Script/Engine.SceneComponent|GetbVisible (Variables|Default|GetPointLight))))))

(fn UpdateAlarmVisuals
  (params (ForceOff :type Boolean) (NewState :type Boolean))
  (bind _isactive (|CallParentFunction ForceOff NewState))
  (Rendering|SetVisibility (Variables|Default|GetPointLight) _isactive false)
  (Rendering|Components|Light|SetIntensity (Variables|Default|GetPointLight) 25.0)
  (bind _audio (Variables|Default|GetAudio)
    (:then
      (if (not _isactive)
        (Audio|Components|Audio|FadeOut _audio 1.0 0.0 "Linear" :!class /Script/Engine.AudioComponent)))
    (:else)))

(fn ConstructionScript
  (|CallParentFunction))
